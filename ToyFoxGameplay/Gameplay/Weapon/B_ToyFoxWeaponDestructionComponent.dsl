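; ToyFox giant pawn component that runs melee weapon attack traces and destroys buildings they hit.
; A swing trace uses server or client-prediction targeting presets, plus landscape targeting for a local impact cue.
; The server feeds hit buildings to the GameState destruction manager, while clients predict their destruction, choosing async or deferred by bounds size.
; Targeting getters derive source location and rotation from clamped aim pitch, with remote view rotation for simulated proxies.
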(event EventBeginPlay
  (bind _returnvalue (Fortnite|Athena|GameState|GetGameStateAthena))
  ;[0 0.1094 0.02157 0.5] BeginPlay: Cache necessary variables.
  (|CallParentFunction)
  (bind _asb_toy_fox_giant_player_pawn (Utilities|Casting|CastToB_ToyFoxGiantPlayerPawn (Components|GetOwner))
    (:then
      (Variables|Transient|SetOwnerPawn (Utilities|ResolveSoftReference _asb_toy_fox_giant_player_pawn))
      (Utilities|IsValid _returnvalue
        (:Is%20Valid
          (bind _output_get (Variables|Transient|SetDestructionManagerComponent (Utilities|ResolveSoftReference (Actor|GetComponentbyClass _returnvalue "/GiantPawnCore/Gameplay/Destruction/B_GiantPawnDestructionManagerGameStateComponent.B_GiantPawnDestructionManagerGameStateComponent_C"))))
          (Utilities|IsValid (Utilities|ResolveSoftReference _output_get)
            (:Is%20Not%20Valid
              (Fortnite|FortShippingLog self "BeginPlay: DestructionManagerComponent is invalid." true))))
        (:Is%20Not%20Valid
          (Fortnite|FortShippingLog self "BeginPlay: GameStateAthena  is invalid." true))))
    (:CastFailed
      (Fortnite|FortShippingLog self "BeginPlay: Owner is invalid." true))))

(event Custom|PerformAttackTrace
  (params (SwingIndex :type Integer))
  (bind _returnvalue (GameInstanceSubsystems|GetTargetingSubsystem))
  ;[0.0241 0.08608 0.15 0.5] PerformAttackTrace: Perform the attack trace for the melee attack.
  ; Invoked from the animation notify. Runs on server and all clients.
  (Variables|Transient|SetLatestSwingIndex SwingIndex)
  (bind _ownerpawn (Variables|Transient|GetOwnerPawn)
    (:then
      (bind (_asynctaskproxy _targetinghandle) (Class|AsyncActionPerformTargeting|PerformTargetingAsyncAction (Utilities|ResolveSoftReference _ownerpawn) (select (Networking|IsDedicatedServer) (Variables|TargetingPreset|GetServerDestructionTargetingPreset) (Variables|TargetingPreset|GetClientPredictionDestructionTargetingPreset)) true)
        (:then
          (Networking|CanExecuteCosmeticEvents
            (:True
              (bind _outhitresults (Utilities|PerformLandscapeTargeting (Variables|TargetingPreset|GetLandscapeTargetingPresets) (Variables|Transient|GetLatestLandscapeTargetingHitResults))
                (:Out
                  (Variables|Transient|SetLatestLandscapeTargetingHitResults _outhitresults)
                  (GameplayCue|ExecuteGameplayCue%28Burst%29Local (Utilities|ResolveSoftReference (Variables|Transient|GetOwnerPawn)) (Utilities|Struct|BreakGameplayCueTag (Variables|GameplayCue|GetLandscapeImpactGameplayCueTag)) (Ability|GameplayCue|MakeGameplayCueParameters :NormalizedMagnitude 0.0 :RawMagnitude 0.0 :Location "0,0,0" :Normal "0,0,0" :GameplayEffectLevel 1 :AbilityLevel (CallFunction|GetLatestSwingIndex) :bReplicateLocationWhenUsingMinimalRepProxy false)))))))
        (:Targeted
          (if (Components|IsComponentBeingDestroyed)
            (Fortnite|FortShippingLog self "PerformAttackTrace: Component is being destroyed." true)
            (else
              (Utilities|IsValid _returnvalue
                (:Is%20Valid
                  (bind _outtargets (TargetingSystem|TargetingResults|GetTargetingResults _returnvalue _targetinghandle))
                  (if (> (Utilities|Array|Length _outtargets) 0)
                    (CallFunction|BuildingDestruction _outtargets)
                    (Default|CallOnTargetingExecuted _targetinghandle (CallFunction|GetLatestSwingIndex))))
                (:Is%20Not%20Valid
                  (Fortnite|FortShippingLog self "PerformAttackTrace: TargetingSubsystem is nullptr." true))))))))
    (:else
      (Fortnite|FortShippingLog self "PerformAttackTrace: OwnerPawn is invalid." true))))

(fn BuildingDestruction
  (params (HitResults :type Array%20of%20Hit%20Result%20Structures))
  ;[0.1347 0 0.15 0.5] BuildingDestruction: Run predictive destruction behavior on the clients. On the server, add the building actors for destruction in the game state component manager.
  (Networking|SwitchHasAuthority
    (:Authority
      (bind _destructionmanagercomponent (Variables|Transient|GetDestructionManagerComponent)
        (:then
          (bind _bvalid (Public|AddBuildingsforDestructionHitResults _destructionmanagercomponent HitResults "ToyFoxWeaponDestructionComponent" (Utilities|ResolveSoftReference (Variables|Transient|GetOwnerPawn))))
          (if _bvalid
            (Public|ForceExecuteDestroyUpdate _destructionmanagercomponent)))
        (:else
          (Fortnite|FortShippingLog self "PerformAttackTrace: DestructionManagerComponent is null." true))))
    (:Remote
      ;;[0.02035 0.08487 0.15 0.5] On the clients, run a prediction trace and apply predictive visual destruction on those building actors.
      (bind _ownerpawn (Variables|Transient|GetOwnerPawn)
        (:then
          (bind _toyfoxgiantpawndestructioncomponent (Variables|/ToyFoxGameplay/Gameplay/B_ToyFoxGiantPlayerPawn.B_ToyFoxGiantPlayerPawn_C|GetToyFoxGiantPawnDestructionComponent (Utilities|ResolveSoftReference _ownerpawn))
            (:then
              (bind _returnvalue (Fortnite|Athena|GameState|GetGameStateAthena))
              (Utilities|IsValid _returnvalue
                (:Is%20Valid
                  (bind _returnvalue_2 (Actor|GetComponentbyClass _returnvalue "/GiantPawnCore/Gameplay/Destruction/B_GiantPawnDestructionManagerGameStateComponent.B_GiantPawnDestructionManagerGameStateComponent_C"))
                  (Utilities|IsValid _returnvalue_2
                    (:Is%20Valid
                      (for _array_element HitResults
                        (bind (_candestroy _asbuildingactor) (Helper|IsActorValidforDestruction _returnvalue_2 (.HitActor (Collision|BreakHitResult _array_element))))
                        (if _candestroy
                          (bind _asbuilding_static_mesh (Utilities|Casting|CastToBuildingSMActor _asbuildingactor))
                          (if (> (Math|Vector|VectorLengthSquared (.BoxExtent (Collision|GetActorBounds _asbuilding_static_mesh true false))) (ScalableFloat|GetCachedScalableFloatValue (Variables|Hotfix|GetHotfix_BuildingBoundsSqrToPredictDeath)))
                            (PredictiveDestruction|PredictDeathforBuildingActorAsync _toyfoxgiantpawndestructioncomponent _asbuilding_static_mesh (Utilities|ResolveSoftReference (Variables|Transient|GetOwnerPawn)) true)
                            (else
                              (PredictiveDestruction|SimplePredictDeathforBuildingActorDeferred _toyfoxgiantpawndestructioncomponent _asbuilding_static_mesh (ScalableFloat|GetCachedScalableFloatValue (Variables|Hotfix|GetHotfix_BuildingPredictDeathDelay))))))))))))
            (:else)))
        (:else)))))

(fn GetLatestLandscapeHitResults
  (return (Variables|Transient|GetLatestLandscapeTargetingHitResults)))

(fn GetLatestSwingIndex
  (return (Variables|Transient|GetLatestSwingIndex)))

(fn GetTargetingAimRotationPitch
  ;[0.07283 0.04486 0.15 0.5] Return the targeting aim pitch after clamping. This is matched with the animation manually.
  (Utilities|FlowControl|Sequence
    (:then_0
      (bind _ownerpawn (Variables|Transient|GetOwnerPawn)
        (:then
          (bind _returnvalue (TargetingSystem|GetPlayerCameraTargetingTransform (Utilities|ResolveSoftReference _ownerpawn)))
          (return true (Math|Float|ClampAngle (.pitch (select (Utilities|Enum|Equal(Enum) (Networking|GetLocalRole (Utilities|ResolveSoftReference _ownerpawn)) "ROLE_SimulatedProxy") (Class|FortPlayerPawn|GetRemoteViewRotation (Utilities|ResolveSoftReference _ownerpawn)) (.rotation _returnvalue))) (ScalableFloat|GetCachedScalableFloatValue (Variables|Hotfix|GetHotfix_MinimumPitchTraceAngle)) (ScalableFloat|GetCachedScalableFloatValue (Variables|Hotfix|GetHotfix_MaximumPitchTraceAngle)))))
        (:else)))
    (:then_1
      (return false 0.0))))

(fn GetTargetingBaseSourceLocation
  ;[0.1425 0.002751 0.15 0.5] Returns the base location which targeting should occur.
  (Utilities|FlowControl|Sequence
    (:then_0
      (bind (_bvalid _pitch) (Targeting|GetTargetingAimRotationPitch))
      (if _bvalid
        (bind _ownerpawn (Variables|Transient|GetOwnerPawn)
          (:then
            (return true (Utilities|Operators|Add (Transformation|GetActorLocation (Utilities|ResolveSoftReference _ownerpawn)) (* (Transformation|GetActorForwardVector (Utilities|ResolveSoftReference _ownerpawn)) (.x (Variables|Settings|GetBaseSourceOffset))) (* (Transformation|GetActorRightVector (Utilities|ResolveSoftReference _ownerpawn)) (.y (Variables|Settings|GetBaseSourceOffset))) (* (Transformation|GetActorUpVector (Utilities|ResolveSoftReference _ownerpawn)) (.z (Variables|Settings|GetBaseSourceOffset))) (Math|Vector|MakeVector 0.0 0.0 (Math|Float|MapRangeClamped _pitch (ScalableFloat|GetCachedScalableFloatValue (Variables|Hotfix|GetHotfix_MinimumPitchTraceAngle)) (ScalableFloat|GetCachedScalableFloatValue (Variables|Hotfix|GetHotfix_MaximumPitchTraceAngle)) (ScalableFloat|GetCachedScalableFloatValue (Variables|Hotfix|GetHotfix_MinimumZOffsetTraceBasedOnPitchAngle)) (ScalableFloat|GetCachedScalableFloatValue (Variables|Hotfix|GetHotfix_MaximumZOffsetTraceBasedOnPitchAngle)))))))
          (:else))))
    (:then_1
      (return false "0, 0, 0"))))

(fn GetTargetingSourceLocation
  ;[0.07748 0.02691 0.15 0.5] GetTargetingSourceLocation: Return the location from where the tracing is done. It combines the source location with the aim rotation pitch value.
  (Utilities|FlowControl|Sequence
    (:then_0
      (bind _ownerpawn (Variables|Transient|GetOwnerPawn)
        (:then
          (bind (_bvalid _location) (Targeting|GetTargetingBaseSourceLocation))
          (if _bvalid
            (bind (_bvalid_2 _pitch) (Targeting|GetTargetingAimRotationPitch))
            (if _bvalid_2
              (return true (+ _location (* (Math|Vector|RotateVectorAroundAxis (Transformation|GetActorForwardVector (Utilities|ResolveSoftReference _ownerpawn)) (* _pitch -1.0) (Transformation|GetActorRightVector (Utilities|ResolveSoftReference _ownerpawn))) (Variables|Settings|GetBoxOverlapHalfExtentX)))))))
        (:else)))
    (:then_1
      (return false "0, 0, 0"))))

(fn GetTargetingSourceRotation
  ;[0.0346 0.03398 0.15 0.5] Returns the targeting source rotation. Uses the aim pitch to determine this value.
  (Utilities|FlowControl|Sequence
    (:then_0
      (bind _ownerpawn (Variables|Transient|GetOwnerPawn)
        (:then
          (bind (_bvalid _pitch) (Targeting|GetTargetingAimRotationPitch))
          (if _bvalid
            (bind _returnvalue (Math|Vector|RotateVectorAroundAxis (Transformation|GetActorForwardVector (Utilities|ResolveSoftReference _ownerpawn)) (* _pitch -1.0) (Transformation|GetActorRightVector (Utilities|ResolveSoftReference _ownerpawn))))
            (return true (Math|Rotator|CombineRotators (Math|Conversions|RotationFromXVector _returnvalue) (Math|Vector|RotatorfromAxisandAngle _returnvalue (Variables|Settings|GetSwingAngle))))))
        (:else)))
    (:then_1
      (return false "0, 0, 0"))))
