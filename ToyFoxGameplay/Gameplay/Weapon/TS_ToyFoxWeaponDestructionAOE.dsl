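; Targeting System AOE task for ToyFox weapon destruction that overrides where the area query originates and how it is oriented.
; GetSourceLocation and GetSourceRotation cast the targeting context source to the giant player pawn and query its weapon destruction component.
; A valid result is returned, with rotation converted to a quaternion.
; Otherwise, each Sequence falls through to the parent implementation.
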(fn GetSourceLocation
  (params (TargetingHandle :type Targeting%20Request%20Handle%20Structure%20%28by%20ref%29))
  (bind _returnvalue (Targeting|GetTargetingSubsystem TargetingHandle))
  (Utilities|FlowControl|Sequence
    (:then_0
      (Utilities|IsValid _returnvalue
        (:Is%20Valid
          (bind _asb_toy_fox_giant_player_pawn (Utilities|Casting|CastToB_ToyFoxGiantPlayerPawn (Utilities|Struct|BreakTargetingSourceContext (TargetingSystem|TargetingTypes|GetTargetingSourceContext _returnvalue TargetingHandle))))
          (Utilities|IsValid (Components|GetWeaponDestructionComponent _asb_toy_fox_giant_player_pawn)
            (:Is%20Valid
              (bind (_bvalid _location) (Targeting|GetTargetingSourceLocation (Components|GetWeaponDestructionComponent _asb_toy_fox_giant_player_pawn)))
              (if _bvalid
                (return _location)))))))
    (:then_1
      (bind _returnvalue_2 (|CallParentFunction TargetingHandle))
      (return _returnvalue_2))))

(fn GetSourceRotation
  (params (TargetingHandle :type Targeting%20Request%20Handle%20Structure%20%28by%20ref%29))
  (bind _returnvalue (Targeting|GetTargetingSubsystem TargetingHandle))
  (Utilities|FlowControl|Sequence
    (:then_0
      (Utilities|IsValid _returnvalue
        (:Is%20Valid
          (bind _asb_toy_fox_giant_player_pawn (Utilities|Casting|CastToB_ToyFoxGiantPlayerPawn (Utilities|Struct|BreakTargetingSourceContext (TargetingSystem|TargetingTypes|GetTargetingSourceContext _returnvalue TargetingHandle))))
          (Utilities|IsValid (Components|GetWeaponDestructionComponent _asb_toy_fox_giant_player_pawn)
            (:Is%20Valid
              (bind (_bvalid _rotation) (Targeting|GetTargetingSourceRotation (Components|GetWeaponDestructionComponent _asb_toy_fox_giant_player_pawn)))
              (if _bvalid
                (return (Math|Conversions|ToQuaternion(Rotator) _rotation))))))))
    (:then_1
      (bind _returnvalue_2 (|CallParentFunction TargetingHandle))
      (return _returnvalue_2))))
